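# Sound cue for a Battle Pass Season 19 GOW quest item, named for Crimson Omen 02, in the quests sound class.
# A single wave player feeds the output with the CogTag sting wave, despite the cue's Crimson Omen name, at slightly raised volume.
# Overridden concurrency allows one instance per owner and blocks new plays while it sounds.
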SoundCue /BattlePassS19/Items/QuestItems/Sounds/GOW_Quest_01_CrimsonOmen02_Cue
root=WavePlayer_2
nodes [{id=WavePlayer_2, properties {soundWaveAssetPtr=/BattlePassS19/Items/QuestItems/Sounds/wav/GOW_Quest01_CogTag_Sting}}]
properties {
 bOverrideConcurrency=true
 maxDistance=2097152
 soundClassObject=&Fort_Quests
 volumeMultiplier=1.1
 concurrencyOverrides {bLimitToOwner=true, maxCount=1, resolutionRule=PreventNew}
}
