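# Sound cue for a Battle Pass Season 19 quest item, named as the second Crimson Omen in a GOW quest.
# A single wave player plays the GOW_Crouch_01 wave through a Modulator node at the root.
# Playback uses the basic foley attenuation within a limited distance and routes to the Fort_Quests sound class.
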SoundCue /BattlePassS19/Items/QuestItems/Sounds/GOW_Quest_02_CrimsonOmen02_Cue
root=Modulator_0
nodes [
 {id=Modulator_0, children [WavePlayer_0]}
 {id=WavePlayer_0, properties {soundWaveAssetPtr=/BattlePassS19/Items/QuestItems/Sounds/wav/GOW_Crouch_01}}
]
properties {
 attenuationSettings=&Foley_Attenuation_Basic_Att
 maxDistance=1536
 soundClassObject=&Fort_Quests
 volumeMultiplier=1
}
