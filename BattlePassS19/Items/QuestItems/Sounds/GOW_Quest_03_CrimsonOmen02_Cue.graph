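# Sound cue for the GOW quest 03 Crimson Omen in Battle Pass S19, playing a chainsaw stinger in the quest sound class.
# A single wave player feeds a modulator that applies a fixed volume boost, offset by a reduced cue volume multiplier.
# Overridden concurrency allows one instance per owner and prevents new plays while it is active.
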SoundCue /BattlePassS19/Items/QuestItems/Sounds/GOW_Quest_03_CrimsonOmen02_Cue
root=Modulator_1
nodes [
 {id=Modulator_1, children [WavePlayer_1], properties {volumeMax=1.2, volumeMin=1.2}}
 {id=WavePlayer_1, properties {soundWaveAssetPtr=/BattlePassS19/Items/QuestItems/Sounds/wav/GOW_Quest03_ChainsawStinger}}
]
properties {
 bOverrideConcurrency=true
 maxDistance=2097152
 soundClassObject=&Fort_Quests
 volumeMultiplier=0.8
 concurrencyOverrides {bLimitToOwner=true, maxCount=1, resolutionRule=PreventNew}
}
