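# Sound cue for a Battle Pass S19 Gears of War quest item (CrimsonOmen02), in the quest sound class.
# A mixer layers a quieter reward-screen chest upgrade sting under one of two randomly chosen, equally weighted Thrashball crowd cheers.
# A concurrency override limits playback to a single instance per owner.
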SoundCue /BattlePassS19/Items/QuestItems/Sounds/GOW_Quest_05_CrimsonOmen02_Cue
root=Mixer_0
nodes [
 {id=Mixer_0, children [Modulator_0,Random_0], properties {inputVolume [1,1]}}
 {id=Modulator_0, children [WavePlayer_3], properties {volumeMax=0.4, volumeMin=0.4}}
 {id=Random_0, children [WavePlayer_0,WavePlayer_2], properties {weights [1,1]}}
 {id=WavePlayer_0, properties {soundWaveAssetPtr=/BattlePassS19/Items/QuestItems/Sounds/wav/GOW_ThrashBall_CrowdCheer_01}}
 {id=WavePlayer_2, properties {soundWaveAssetPtr=/BattlePassS19/Items/QuestItems/Sounds/wav/GOW_ThrashBall_CrowdCheer_02}}
 {id=WavePlayer_3, properties {soundWaveAssetPtr=/BattlePassS19/Items/QuestItems/Sounds/wav/Fort_RewardScreen_TeamScore_ChestLootUpgrade}}
]
properties {
 bOverrideConcurrency=true
 maxDistance=2097152
 soundClassObject=&Fort_Quests
 volumeMultiplier=0.88
 concurrencyOverrides {bLimitToOwner=true, maxCount=1}
}
